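# Physics asset for the CorvidStomp pickaxe mesh, providing physics bodies for its dynamic feather and bond joints.
# Fourteen feather bones and a bond bone each get a capsule body, constrained directly to a shared feather offset body in a flat star layout rather than chains.
# All constraints carry tight swing limit angles, and the feathers also lock twist, while the bond joint does not.
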PhysicsAsset /BRCosmetics/Weapons/FORT_Melee/Pickaxe_CorvidStomp/Meshes/CorvidStomp_Axe_PhysicsAsset
bodies {
 dyn_bond_joint [{shapeType=Capsule, length=2.5, radius=0.5, shapeName=dyn_bond_joint_capsule, center=<2.6495 0 0>, rotation=<rot 90 -0.0004 -0.0006>}]
 dyn_feather_10_joint [{shapeType=Capsule, length=5, radius=1.5, shapeName=dyn_feather_10_joint_capsule, center=<4.5921 0 0>, rotation=<rot 90 -160.5289 -160.5292>}]
 dyn_feather_11_joint [{shapeType=Capsule, length=3.5, radius=1, shapeName=dyn_feather_11_joint_capsule, center=<3.0564 0 0>, rotation=<rot 90 179.9209 179.9208>}]
 dyn_feather_12_joint [{
   shapeType=Capsule
   length=3.5
   radius=1.5
   shapeName=dyn_feather_12_joint_capsule
   center=<3.6215 0 0>
   rotation=<rot 90 -179.9987 -179.9989>
  }]
 dyn_feather_13_joint [{shapeType=Capsule, length=8.5, radius=1.5, shapeName=dyn_feather_13_joint_capsule, center=<6.2571 0 0>, rotation=<rot 90 179.9975 179.9974>}]
 dyn_feather_14_joint [{shapeType=Capsule, length=9, radius=2, shapeName=dyn_feather_14_joint_capsule, center=<7.536 0 0>, rotation=<rot 90 -90 -90.0002>}]
 dyn_feather_1_joint [{shapeType=Capsule, length=3.5, radius=0.5, shapeName=dyn_feather_1_joint_capsule, center=<2.7325 0 0>, rotation=<rot 90 90 90.0003>}]
 dyn_feather_2_joint [{shapeType=Capsule, length=5, radius=1, shapeName=dyn_feather_2_joint_capsule, center=<4.1685 0 0>, rotation=<rot -90 90 -90.0004>}]
 dyn_feather_3_joint [{shapeType=Capsule, length=3, radius=1, shapeName=dyn_feather_3_joint_capsule, center=<2.708 0 0>, rotation=<rot 90 -166.7373 -166.7372>}]
 dyn_feather_4_joint [{shapeType=Capsule, length=3, radius=1, shapeName=dyn_feather_4_joint_capsule, center=<3.3429 0 0>, rotation=<rot -90 165.1488 -165.1492>}]
 dyn_feather_5_joint [{shapeType=Capsule, length=6.5, radius=1.5, shapeName=dyn_feather_5_joint_capsule, center=<4.9947 0 0>, rotation=<rot 90 180 -179.9997>}]
 dyn_feather_6_joint [{shapeType=Capsule, length=6.5, radius=1.5, shapeName=dyn_feather_6_joint_capsule, center=<6.1142 0 0>, rotation=<rot -90 0 -0.0003>}]
 dyn_feather_7_joint [{shapeType=Capsule, length=2, radius=0.5, shapeName=dyn_feather_7_joint_capsule, center=<1.9425 0 0>, rotation=<rot -90 0 -0.0003>}]
 dyn_feather_8_joint [{shapeType=Capsule, length=7.5, radius=1.5, shapeName=dyn_feather_8_joint_capsule, center=<5.9358 0 0>, rotation=<rot 90 90 90.0001>}]
 dyn_feather_9_joint [{shapeType=Capsule, length=4, radius=1, shapeName=dyn_feather_9_joint_capsule, center=<3.3423 0 0>, rotation=<rot -90 179.9802 -179.98>}]
 dyn_feather_offset_joint [{
   shapeType=Capsule
   length=2.0229
   radius=1
   shapeName=dyn_feather_offset_joint_capsule
   center=<0.0004 -0.1132 -0.1878>
   rotation=<rot -4.6002 -90 0.0013>
  }]
}
constraints [
 {bone1Name=dyn_bond_joint, bone2Name=dyn_feather_offset_joint, swing1LimitDegrees=10, swing2LimitDegrees=10, twistLimitDegrees=10}
 {
  bone1Name=dyn_feather_10_joint
  bone2Name=dyn_feather_offset_joint
  swing1LimitDegrees=10
  swing2LimitDegrees=10
  twistLimitDegrees=10
  twistMotion=Locked
 }
 {
  bone1Name=dyn_feather_11_joint
  bone2Name=dyn_feather_offset_joint
  swing1LimitDegrees=10
  swing2LimitDegrees=10
  twistLimitDegrees=10
  twistMotion=Locked
 }
 {
  bone1Name=dyn_feather_12_joint
  bone2Name=dyn_feather_offset_joint
  swing1LimitDegrees=10
  swing2LimitDegrees=10
  twistLimitDegrees=10
  twistMotion=Locked
 }
 {
  bone1Name=dyn_feather_13_joint
  bone2Name=dyn_feather_offset_joint
  swing1LimitDegrees=10
  swing2LimitDegrees=10
  twistLimitDegrees=10
  twistMotion=Locked
 }
 {
  bone1Name=dyn_feather_14_joint
  bone2Name=dyn_feather_offset_joint
  swing1LimitDegrees=10
  swing2LimitDegrees=10
  twistLimitDegrees=10
  twistMotion=Locked
 }
 {
  bone1Name=dyn_feather_1_joint
  bone2Name=dyn_feather_offset_joint
  swing1LimitDegrees=10
  swing2LimitDegrees=10
  twistLimitDegrees=10
  twistMotion=Locked
 }
 {
  bone1Name=dyn_feather_2_joint
  bone2Name=dyn_feather_offset_joint
  swing1LimitDegrees=10
  swing2LimitDegrees=10
  twistLimitDegrees=10
  twistMotion=Locked
 }
 {
  bone1Name=dyn_feather_3_joint
  bone2Name=dyn_feather_offset_joint
  swing1LimitDegrees=10
  swing2LimitDegrees=10
  twistLimitDegrees=10
  twistMotion=Locked
 }
 {
  bone1Name=dyn_feather_4_joint
  bone2Name=dyn_feather_offset_joint
  swing1LimitDegrees=10
  swing2LimitDegrees=10
  twistLimitDegrees=10
  twistMotion=Locked
 }
 {
  bone1Name=dyn_feather_5_joint
  bone2Name=dyn_feather_offset_joint
  swing1LimitDegrees=10
  swing2LimitDegrees=10
  twistLimitDegrees=10
  twistMotion=Locked
 }
 {
  bone1Name=dyn_feather_6_joint
  bone2Name=dyn_feather_offset_joint
  swing1LimitDegrees=10
  swing2LimitDegrees=10
  twistLimitDegrees=10
  twistMotion=Locked
 }
 {
  bone1Name=dyn_feather_7_joint
  bone2Name=dyn_feather_offset_joint
  swing1LimitDegrees=10
  swing2LimitDegrees=10
  twistLimitDegrees=10
  twistMotion=Locked
 }
 {
  bone1Name=dyn_feather_8_joint
  bone2Name=dyn_feather_offset_joint
  swing1LimitDegrees=10
  swing2LimitDegrees=10
  twistLimitDegrees=10
  twistMotion=Locked
 }
 {
  bone1Name=dyn_feather_9_joint
  bone2Name=dyn_feather_offset_joint
  swing1LimitDegrees=10
  swing2LimitDegrees=10
  twistLimitDegrees=10
  twistMotion=Locked
 }
]
